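# Niagara system for the Colorsmoke vehicle boost cosmetic, emitting colored smoke and bright bursts from the car's skeletal-mesh sockets.
# Both emitters partition particles across the ports, apply a car offset and inherit source movement.
# MainTrail streams rising flipbook sprites, recolored over life by per-paint-variant curves on several color channels.
# Burst loops probabilistic bursts plus a sparse secondary spawn group with separate motion, alternating two colors per loop for some paints.
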NiagaraSystem /VehicleCosmetics/Boosts/Boost_Colorsmoke/FX/NS_Boost_Colorsmoke
dependencies {
 usedDataInterfaces [
  &NiagaraDataInterfaceArrayColor
  &NiagaraDataInterfaceArrayFloat2
  &NiagaraDataInterfaceCurve
  &NiagaraDataInterfaceVectorCurve
  &NiagaraDataInterfaceVector2DCurve
  &NiagaraDataInterfaceArrayFloat
  &NiagaraDataInterfaceSpriteRendererInfo
  &NiagaraDataInterfaceColorCurve
  &NiagaraDataInterfaceSkeletalMesh
 ]
 usedDynamicInputs []
 usedModules [
  &self:SystemSpawnScript.SystemScriptSource.SystemScriptGraph.NiagaraNodeAssignment_0.SetVariables_#1
  &SystemState
 ]
}
emitters {
 NE_CS_Burst {
  emitterUpdateScript [
   &EmitterState(Life Cycle Mode=::NewEnumerator1, Loop Duration=0.55, UseLoopDelay=false)
   &VC_FilteredSocketCount()
   &SpawnBurst_Instantaneous(Spawn Count=38, Spawn Probability=0.6, Use Spawn Probability=true, UseLoopCountLimit=false)
   &SpawnRate(Spawn Probability=0.4, SpawnGroup=1, SpawnRate=6, Use Spawn Probability=true)
  ]
  particleSpawnScript [
   /SetVariables(Particles.VisibilityTag=$Emitter.SpawnGroup) {isSetParameters=true}
   &VC_PaintedVariantsOverride(
    Black=<rgba 0.7341 0.7341 0.7341 1>
    BurntSienna=<rgba 1 0.6805 0.1078 1>
    Cobalt=<rgba 0.5042 0.15 1 1>
    Crimson=&SelectLinearColorFromArray(
     Array Sampling Mode=::NewEnumerator1
     Direct Array Index=&ModuloInt(A=$Emitter.LoopCount, B=2)
     Color Selection Array=DataInterface() {propertyValues {ColorData [<rgba 1 0.55 0 1>, <rgba 1 0.08 0.08 1>]}}
     Random Seed=?
    )
    DefaultColor=<rgba 0.2233 0.025 0.5 1>
    ForestGreen=&SelectLinearColorFromArray(
     Array Sampling Mode=::NewEnumerator1
     Direct Array Index=&ModuloInt(A=$Emitter.LoopCount, B=2)
     Color Selection Array=DataInterface() {propertyValues {ColorData [<rgba 0.7086 1 0.7231 1>, <rgba 0.04 0.5 0.063 1>]}}
     Random Seed=?
    )
    Grey=<rgba 0.2 0.2 0.2 1>
    Lime=&SelectLinearColorFromArray(
     Array Sampling Mode=::NewEnumerator1
     Direct Array Index=&ModuloInt(A=$Emitter.LoopCount, B=2)
     Color Selection Array=DataInterface() {propertyValues {ColorData [<rgba 0.95 1 0.9 1>, <rgba 0.29 0.58 0 1>]}}
     Random Seed=?
    )
    Orange=<rgba 1 0.8949 0 1>
    Pink=<rgba 1 0.3 0.7481 1>
    Purple=&SelectLinearColorFromArray(
     Array Sampling Mode=::NewEnumerator1
     Direct Array Index=&ModuloInt(A=$Emitter.LoopCount, B=2)
     Color Selection Array=DataInterface() {propertyValues {ColorData [<rgba 0.6296 0.0867 1 1>, <rgba 1 0.5188 0.7691 1>]}}
     Random Seed=?
    )
    R1 Black:LinearColor=<rgba 0.2517 0.2517 0.2517 0.8>
    R1 BurntSienna:LinearColor=<rgba 0.3177 0.0981 0.0392 0.5>
    R1 Cobalt:LinearColor=<rgba 0 0.0083 0.45 0.5>
    R1 Crimson:LinearColor=<rgba 0.4 0.08 0.08 0.5>
    R1 DefaultColor=<rgba 0.337 0.3015 0.67 1>
    R1 ForestGreen:LinearColor=<rgba 0.2619 0.45 0.2714 0.5>
    R1 Grey:LinearColor=<rgba 0.35 0.35 0.35 0.5>
    R1 Lime:LinearColor=<rgba 0.3204 0.5 0.1408 0.5>
    R1 Orange:LinearColor=<rgba 0.4 0.36 0 0.5>
    R1 Pink:LinearColor=<rgba 0.5 0.4 0.4636 0.5>
    R1 Purple:LinearColor=<rgba 0.25 0.1297 0.1923 0.5>
    R1 Saffron:LinearColor=<rgba 0.45 0.345 0 0.5>
    R1 SkyBlue:LinearColor=<rgba 0.1928 0.3307 0.4 0.5>
    R1 TitaniumWhite:LinearColor=<rgba 0.45 0.45 0.45 1>
    Saffron=&SelectLinearColorFromArray(
     Array Sampling Mode=::NewEnumerator1
     Direct Array Index=&ModuloInt(A=$Emitter.LoopCount, B=2)
     Color Selection Array=DataInterface() {propertyValues {ColorData [<rgba 0.829 0.6356 0 1>, <rgba 0.7447 0.7447 0.7447 1>]}}
     Random Seed=?
    )
    SkyBlue=<rgba 0.2 0.7323 1 1>
    TitaniumWhite=<rgba 1 1 1 1>
    Use R1 Painted Overrides=true
   )
   &InitializeParticle(
    Color Mode=::NewEnumerator0
    Lifetime=&RandomRangeFloat(Minimum=0.2, Maximum=0.45)
    Sprite Rotation Mode=::NewEnumerator1
    Sprite Size=&SelectVector2DFromArray(
     Array Sampling Mode=::NewEnumerator0
     Direct Array Index=?
     Vector 2D Selection Array=DataInterface() {propertyValues {FloatData [<1 5>, <1.5 13>, <2 10>, <1 1>, <1 1>, <1.5 10>]}}
     Random Seed=0
    )
    Sprite Size Mode=::NewEnumerator4
    UsePositionOffset=false
    Write Material Random=true
   )
   &PartitionParticles(Number Of Partitions=$Emitter.PortCount)
   &SkeletalMeshLocation(
    Bone / Socket Index=$Particles.PartitionParticles.Partition
    Bone / Socket Sampling Mode=::NewEnumerator5
    MaskSpawnGroup=false
    Mesh Sampling Type=::NewEnumerator6
    Sampled Position Offset=&AddVectorToPosition(
     Position=&MergeHelper_VectorSwitch(0=<-120 0 -5>, 1=<0 0 0>, 2=?, 3=?, Index=&SpawnGroup(), Num=::NewEnumerator0)
     Vector=$User.DTCarOffset
    )
    Skeletal Mesh=$System.SkeletalMesh
   )
   &ShapeLocation(
    Cylinder Distribution=::NewEnumerator2
    Cylinder Height=150
    Cylinder Height Midpoint:NiagaraFloat=&ReturnNormalizedExecIndex(Normalized Index Scale=0.7, Normalized Index Includes One=false)
    Cylinder Radius=&ReturnNormalizedExecIndex(Normalized Index Scale=0.1, Normalized Index Includes One=false)
    MaskSpawnGroup=false
    Rotation Mode=::NewEnumerator2
    Shape Primitive=::NewEnumerator1
    U Position=&ReturnNormalizedExecIndex(Normalized Index Scale=0.45, Normalized Index Includes One=false)
    UseOwnerScale=true
    V Position=1
    Yaw / Pitch / Roll=<0 0.25 0>
   ) {module=ShapeLocation001}
   &AddVelocity(
    Cone Angle=&MergeHelper_FloatSwitch(0=60, 1=0, 2=?, 3=?, Index=&SpawnGroup(), Num=::NewEnumerator0)
    Cone Axis=<1 0 -0.2>
    Inner Cone Angle:NiagaraFloat=&MergeHelper_FloatSwitch(0=55, 1=0, 2=?, 3=?, Index=&SpawnGroup(), Num=::NewEnumerator0)
    Module.UseConeFalloff=true
    Module.UseVelDistribution=false
    Override Seed=false
    Speed Falloff From Cone Axis=0
    Velocity Mode=::NewEnumerator2
    Velocity Speed=&MergeHelper_FloatSwitch(0=-1000, 1=-650, 2=?, 3=?, Index=&SpawnGroup(), Num=::NewEnumerator0)
   )
  ]
  particleUpdateScript [
   &ParticleState()
   &GravityForce(Gravity=&MergeHelper_VectorSwitch(0=<0 0 -800>, 1=<0 0 -400>, 2=?, 3=?, Index=&SpawnGroup(), Num=::NewEnumerator0))
   &Drag(
    Drag=&MergeHelper_FloatSwitch(
     0=&RandomRangeFloat(Minimum=8, Maximum=8)
     1=&RandomRangeFloat(Minimum=0, Maximum=2)
     2=?
     3=?
     Index=&SpawnGroup()
     Num=::NewEnumerator0
    )
    Use Linear Drag=true
    Use Rotational Drag=false
   ) {module=Drag001}
   &InheritSourceMovement(ApplyForce=false, ApplyPosition=true, ApplyPositionFalloff=false)
   &SolveForcesAndVelocity(Clamp Velocity=false, Limit Acceleration=false)
   &ScaleSpriteSize(
    Scale Sprite Size Mode=::NewEnumerator0
    Uniform Scale Factor=&MakeFloatFromVector(
     Vector=&VectorFromCurve(
      VectorCurve=DataInterface() {propertyValues {
        XCurve {keys [
          {
           arriveTangent=-1
           arriveTangentWeight=0.2
           interpMode=RCIM_Cubic
           leaveTangent=-1
           leaveTangentWeight=0.2828
           tangentMode=RCTM_User
           time=0.4
           value=1
          }
          {arriveTangentWeight=0.2, interpMode=RCIM_Cubic, leaveTangentWeight=0.2, tangentMode=RCTM_User, time=1, value=0}
         ]}
        YCurve {keys [
          <at 0 1 RCIM_Cubic>
          <at 0.1 5 RCIM_Cubic>
          {arriveTangent=-5.5556, interpMode=RCIM_Cubic, leaveTangent=-5.5556, time=0.4, value=1}
          <at 1 0 RCIM_Cubic>
         ]}
        ZCurve {keys [{arriveTangent=-2.5, interpMode=RCIM_Cubic, time=0, value=0}, <at 0.4 0 RCIM_Cubic>]}
       }}
      Scale Curve=&VectorFromFloat(Value=1.5)
     )
     Channel=::NewEnumerator0
    )
   )
   &ScaleColor(
    Scale Alpha=&FloatFromCurve(
     FloatCurve=DataInterface() {propertyValues {Curve {
        defaultValue=1
        keys [{interpMode=RCIM_Cubic, tangentMode=RCTM_User, time=0, value=1}, {interpMode=RCIM_Cubic, tangentMode=RCTM_User, time=0.7, value=0}]
       }}}
     Scale Curve=2
    )
    Scale RGB=&Multiply_VectorByFloat(Vector=<2 2 2>, Float=&RandomRangeFloat(Minimum=2, Maximum=20))
    ScaleA=true
    ScaleRGB=true
   )
   /SetVariables(
    Particles.LightningSize:Vector2f=&Vector2DFromCurve(
     Vector2Curve=DataInterface() {propertyValues {
       XCurve {keys [{arriveTangent=0.2281, interpMode=RCIM_Cubic, time=0, value=0.8}, {interpMode=RCIM_Cubic, leaveTangent=0.8667, time=1, value=1.7}]}
       YCurve {keys [
         {arriveTangent=0.1081, interpMode=RCIM_Cubic, leaveTangent=0.1081, tangentMode=RCTM_User, time=0, value=0.8}
         {arriveTangent=-0.042, interpMode=RCIM_Cubic, leaveTangent=-0.042, tangentMode=RCTM_User, time=1, value=1.1}
        ]}
      }}
     ScaleCurve=&RandomRangeVector2D(Minimum=<60 85>, Maximum=<80 95>)
    )
    Particles.PositionBursts:NiagaraPosition=&AddVectorToPosition(
     Position=$Particles.Position
     Vector=/VehicleCosmetics/FX/Niagara/Modules/VC_ConvertVectorToWorld(Vector:Vector3f=&MakeVector(
       X=&SelectFloatFromArray(
        Array Sampling Mode=::NewEnumerator0
        Direct Array Index=?
        Float Selection Array=DataInterface() {propertyValues {FloatData [80,-140]}}
        Random Seed=0
       )
       Y=0
       Z=0
      ))
    )
    Particles.ScaledR1:LinearColor=&MakeLinearColorFromVectorAndFloat(
     "Vector (RGB)"=&Multiply_VectorByFloat(Vector=&MakeVectorFromLinearColorRGB(Color=$Particles.R1 Color), Float=1.5)
     "Float (Alpha)"=&FloatFromCurve(
      FloatCurve=DataInterface() {propertyValues {Curve {keys [
          {arriveTangent=-7.8312, interpMode=RCIM_Cubic, leaveTangent=-7.8312, tangentMode=RCTM_User, time=0, value=1}
          {interpMode=RCIM_Cubic, tangentMode=RCTM_User, time=0.6, value=0.01}
         ]}}}
      Scale Curve=4
     )
    )
   ) {isSetParameters=true}
   &SubUVAnimation(
    Play Rate:NiagaraFloat=&RandomRangeFloat(Minimum=0.4, Maximum=1.5)
    Playback Scale Curve=DataInterface() {propertyValues {Curve {keys [
        {arriveTangent=1.2656, interpMode=RCIM_Cubic, leaveTangent=1.2656, leaveTangentWeight=0.4033, tangentMode=RCTM_User, time=0, value=0}
        {arriveTangentWeight=0.25, interpMode=RCIM_Cubic, tangentMode=RCTM_User, time=0.75, value=1}
       ]}}}
    Random Start Frame=true
    "Sprite Renderer "=DataInterface() {propertyValues {SpriteRenderer=&self:NE_CS_Burst_0.NiagaraSpriteRendererProperties_0}}
    Start Frame Offset:NiagaraInt32=&RandomRangeInt(Minimum=1, Maximum=5)
    SubUV Animation Mode:ENiagara_SubUVLookupModeV2=::NewEnumerator3
    UseEndFrame=false
    UsePlayRateScale=true
    UseStartFrame=false
   )
  ]
  rendererClasses [&NiagaraSpriteRendererProperties]
  renderers [
   {rendererClass=&NiagaraSpriteRendererProperties}
   {rendererClass=&NiagaraSpriteRendererProperties, rendererIndex=1}
  ]
 }
 NE_CS_MainTrail {
  emitterUpdateScript [
   &EmitterState()
   &VC_FilteredSocketCount()
   &SpawnRate(SpawnRate=70, Use Spawn Probability=true)
  ]
  particleSpawnScript [
   /SetVariables(Particles.VisibilityTag=$Emitter.SpawnGroup) {isSetParameters=true}
   &VC_PaintedVariantsOverride(
    R2 Black:LinearColor=<rgba 0.16 0.16 0.16 1>
    R2 DefaultColor:LinearColor=<rgba 0.9348 0.5504 0.9896 1>
    Use R1 Painted Overrides=true
    Use R2 Painted Overrides=true
   ) {enabled=false}
   &InitializeParticle(
    Color=$Particles.Color
    Lifetime Max=0.7
    Lifetime Min=0.6
    Lifetime Mode=::NewEnumerator1
    Sprite Rotation Angle=&RandomRangeFloat(Minimum=-0.2, Maximum=0.35)
    Sprite Rotation Mode=::NewEnumerator1
    Sprite Size Mode=::NewEnumerator3
    Uniform Sprite Size=250
    UsePositionOffset=false
    Write Material Random=true
   )
   &PartitionParticles(Number Of Partitions=$Emitter.PortCount)
   &SkeletalMeshLocation(
    Bone / Socket Index=$Particles.PartitionParticles.Partition
    Bone / Socket Sampling Mode=::NewEnumerator5
    MaskSpawnGroup=false
    Mesh Sampling Type=::NewEnumerator6
    Sampled Position Offset=&AddVectorToPosition(Position=<0 0 0>, Vector=$User.DTCarOffset)
    Skeletal Mesh=$System.SkeletalMesh
   )
   &AddVelocity(Velocity=<-1000 0 0>)
  ]
  particleUpdateScript [
   &ParticleState()
   &VC_PaintedVariantsOverride(
    Black=&ColorFromCurve(ColorCurve=DataInterface() {propertyValues {
       AlphaCurve {keys [<at 0 0 RCIM_Cubic>,<at 0.5769 1.5 RCIM_Cubic>]}
       BlueCurve {keys [<at 0.1205 0.014 RCIM_Cubic>]}
       GreenCurve {keys [<at 0.1205 0.014 RCIM_Cubic>]}
       RedCurve {keys [<at 0.1205 0.014 RCIM_Cubic>]}
      }})
    BurntSienna=&ColorFromCurve(ColorCurve=DataInterface() {propertyValues {
       AlphaCurve {keys [<at 0.0646 0>,<at 0.1641 1 RCIM_Cubic>,<at 0.84 0>]}
       BlueCurve {keys [<at 0.1749 0.0058 RCIM_Cubic>,<at 0.5744 0.0014 RCIM_Cubic>]}
       GreenCurve {keys [<at 0.1749 0.0293 RCIM_Cubic>,<at 0.5744 0.025 RCIM_Cubic>]}
       RedCurve {keys [<at 0.1749 0.18 RCIM_Cubic>,<at 0.5744 0.0207 RCIM_Cubic>]}
      }})
    Cobalt=&ColorFromCurve(ColorCurve=DataInterface() {propertyValues {
       AlphaCurve {keys [<at 0 0 RCIM_Cubic>,<at 0.5462 1 RCIM_Cubic>,<at 0.8973 0>]}
       BlueCurve {keys [<at 0.1503 2 RCIM_Cubic>,<at 0.5806 0.08>]}
       GreenCurve {keys [<at 0.1503 0 RCIM_Cubic>,<at 0.5806 0.0216>]}
       RedCurve {keys [<at 0.1503 0.0006 RCIM_Cubic>,<at 0.5806 0.06>]}
      }})
    Crimson=&ColorFromCurve(ColorCurve=DataInterface() {propertyValues {
       AlphaCurve {keys [<at 0.0654 0>,<at 0.1921 1 RCIM_Cubic>,<at 0.8687 0>]}
       BlueCurve {keys [<at 0.2019 0.01 RCIM_Cubic>,<at 0.5683 0.02 RCIM_Cubic>]}
       GreenCurve {keys [<at 0.2019 0.01 RCIM_Cubic>,<at 0.5683 0.04 RCIM_Cubic>]}
       RedCurve {keys [<at 0.2019 0.2 RCIM_Cubic>,<at 0.5683 0.038 RCIM_Cubic>]}
      }})
    DefaultColor=&ColorFromCurve(ColorCurve=DataInterface() {propertyValues {
       AlphaCurve {
        defaultValue=0
        keys [
         <at 0 0 RCIM_Cubic>
         <at 0.2 1>
         {
          arriveTangent=-0.0405
          arriveTangentWeight=0.2004
          interpMode=RCIM_Cubic
          leaveTangent=-0.0405
          leaveTangentWeight=0.0998
          tangentMode=RCTM_User
          time=0.8527
          value=0
         }
        ]
       }
       BlueCurve {keys [<at 0.1753 0.5156 RCIM_Cubic>,<at 0.5739 0.1>]}
       GreenCurve {keys [<at 0.1753 0.1425 RCIM_Cubic>,<at 0.5739 0.044>]}
       RedCurve {keys [<at 0.1753 0.1793 RCIM_Cubic>,<at 0.5739 0.0897>]}
      }})
    ForestGreen=&ColorFromCurve(ColorCurve=DataInterface() {propertyValues {
       AlphaCurve {keys [<at 0.0381 0>,<at 0.1921 1.3 RCIM_Cubic>,<at 0.8655 0>]}
       BlueCurve {keys [<at 0.1872 0 RCIM_Cubic>,<at 0.5708 0.0333>,<at 0.679 0.0373>]}
       GreenCurve {keys [<at 0.1872 0.28 RCIM_Cubic>,<at 0.5708 0.04>,<at 0.679 0.04>]}
       RedCurve {keys [<at 0.1872 0.0467 RCIM_Cubic>,<at 0.5708 0>,<at 0.679 0.024>]}
      }})
    Grey=&ColorFromCurve(ColorCurve=DataInterface() {propertyValues {
       AlphaCurve {keys [
         <at 0 0 RCIM_Cubic>
         {arriveTangent=1.4839, interpMode=RCIM_Cubic, leaveTangent=1.4839, time=0.4899, value=2.5}
         <at 0.6065 0.9>
         <at 0.9085 0>
        ]}
       BlueCurve {keys [<at 0.3574 0.12 RCIM_Cubic>,<at 0.5955 0.035>]}
       GreenCurve {keys [<at 0.3574 0.12 RCIM_Cubic>,<at 0.5955 0.035>]}
       RedCurve {keys [<at 0.3574 0.12 RCIM_Cubic>,<at 0.5955 0.035>]}
      }})
    Lime=&ColorFromCurve(ColorCurve=DataInterface() {propertyValues {
       AlphaCurve {keys [
         <at 0.082 0 RCIM_Cubic>
         {arriveTangent=2.8732, interpMode=RCIM_Cubic, leaveTangent=2.8732, time=0.1859, value=1}
         <at 0.4301 1>
         <at 0.754 0>
        ]}
       BlueCurve {keys [<at 0.2222 0 RCIM_Cubic>,<at 0.5391 0.061 RCIM_Cubic>]}
       GreenCurve {keys [<at 0.2222 0.7917 RCIM_Cubic>,<at 0.5391 0.1476 RCIM_Cubic>]}
       RedCurve {keys [<at 0.2222 0.1696 RCIM_Cubic>,<at 0.5391 0.0081 RCIM_Cubic>]}
      }})
    Orange=&ColorFromCurve(ColorCurve=DataInterface() {propertyValues {
       AlphaCurve {keys [
         <at 0.0486 0 RCIM_Cubic>
         {arriveTangent=2.6214, interpMode=RCIM_Cubic, leaveTangent=2.6214, time=0.1859, value=1}
         <at 0.4301 1>
         <at 0.7524 0>
        ]}
       BlueCurve {keys [<at 0.1828 0.0008 RCIM_Cubic>,<at 0.5088 0 RCIM_Cubic>]}
       GreenCurve {keys [<at 0.1828 0.2332 RCIM_Cubic>,<at 0.5088 0.09 RCIM_Cubic>]}
       RedCurve {keys [<at 0.1828 0.6146 RCIM_Cubic>,<at 0.5088 0.075 RCIM_Cubic>]}
      }})
    Pink=&ColorFromCurve(ColorCurve=DataInterface() {propertyValues {
       AlphaCurve {keys [
         <at 0.0629 0 RCIM_Cubic>
         {arriveTangent=2.7237, interpMode=RCIM_Cubic, leaveTangent=2.7237, time=0.1859, value=1}
         <at 0.4301 1>
         <at 0.8273 0>
        ]}
       BlueCurve {keys [<at 0.1828 0.2344 RCIM_Cubic>,<at 0.5741 0.04 RCIM_Cubic>]}
       GreenCurve {keys [<at 0.1828 0.093 RCIM_Cubic>,<at 0.5741 0.06 RCIM_Cubic>]}
       RedCurve {keys [<at 0.1828 0.3 RCIM_Cubic>,<at 0.5741 0.1 RCIM_Cubic>]}
      }})
    Purple=&ColorFromCurve(ColorCurve=DataInterface() {propertyValues {
       AlphaCurve {keys [<at 0.0508 0 RCIM_Cubic>,<at 0.2922 1 RCIM_Cubic>,<at 0.949 0>]}
       BlueCurve {keys [<at 0.2333 0.2 RCIM_Cubic>,<at 0.6077 0.0365 RCIM_Cubic>]}
       GreenCurve {keys [<at 0.2333 0.01 RCIM_Cubic>,<at 0.6077 0.0043 RCIM_Cubic>]}
       RedCurve {keys [<at 0.2333 0.1089 RCIM_Cubic>,<at 0.6077 0.0214 RCIM_Cubic>]}
      }})
    R1 Black:LinearColor=&ColorFromCurve(ColorCurve=DataInterface() {propertyValues {
       AlphaCurve {keys [<at 0.1097 0 RCIM_Cubic>,<at 0.2745 1 RCIM_Cubic>,<at 0.6836 0>]}
       BlueCurve {keys [<at 0.1091 0.01 RCIM_Cubic>,<at 0.2892 0.1771 RCIM_Cubic>]}
       GreenCurve {keys [<at 0.1091 0.01 RCIM_Cubic>,<at 0.2892 0.1771 RCIM_Cubic>]}
       RedCurve {keys [<at 0.1091 0.01 RCIM_Cubic>,<at 0.2892 0.1771 RCIM_Cubic>]}
      }})
    R1 BurntSienna:LinearColor=&ColorFromCurve(ColorCurve=DataInterface() {propertyValues {
       AlphaCurve {keys [<at 0 0>,<at 0.469 1 RCIM_Cubic>,<at 0.6525 0 RCIM_Cubic>]}
       BlueCurve {keys [<at 0 0.0178 RCIM_Cubic>,<at 0.6548 0.0108 RCIM_Cubic>]}
       GreenCurve {keys [<at 0 0.0908 RCIM_Cubic>,<at 0.6548 0.6927 RCIM_Cubic>]}
       RedCurve {keys [<at 0 0.5573 RCIM_Cubic>,<at 0.6548 0.57 RCIM_Cubic>]}
      }})
    R1 Cobalt:LinearColor=&ColorFromCurve(ColorCurve=DataInterface() {propertyValues {
       AlphaCurve {keys [<at 0.1174 0 RCIM_Cubic>,<at 0.2528 1 RCIM_Cubic>,<at 0.6105 0>]}
       BlueCurve {keys [<at 0.1753 1 RCIM_Cubic>,<at 0.6077 3>]}
       GreenCurve {keys [<at 0.1753 0.4069 RCIM_Cubic>,<at 0.6077 0>]}
       RedCurve {keys [<at 0.1753 0.3646 RCIM_Cubic>,<at 0.6077 2.5078>]}
      }})
    R1 Crimson:LinearColor=&ColorFromCurve(ColorCurve=DataInterface() {propertyValues {
       AlphaCurve {keys [
         <at 0.147 0 RCIM_Cubic>
         {arriveTangent=3.9441, interpMode=RCIM_Cubic, leaveTangent=3.9441, time=0.2808, value=1}
         <at 0.4005 1>
         <at 0.6074 0>
        ]}
       BlueCurve {keys [<at 0.1753 0.1286 RCIM_Cubic>,<at 0.6077 0>]}
       GreenCurve {keys [<at 0.1753 0.1763 RCIM_Cubic>,<at 0.6077 1.4989>]}
       RedCurve {keys [<at 0.1753 1 RCIM_Cubic>,<at 0.6077 1.5>]}
      }})
    R1 DefaultColor=&ColorFromCurve(ColorCurve=DataInterface() {propertyValues {
       AlphaCurve {keys [<at 0.0817 0 RCIM_Cubic>,<at 0.2357 1 RCIM_Cubic>,<at 0.6821 0>]}
       BlueCurve {keys [<at 0.1753 1 RCIM_Cubic>,<at 0.6077 2>]}
       GreenCurve {keys [<at 0.1753 0.8125 RCIM_Cubic>,<at 0.6077 0.6875>]}
       RedCurve {keys [<at 0.1753 0.831 RCIM_Cubic>,<at 0.6077 1.7446>]}
      }})
    R1 ForestGreen:LinearColor=&ColorFromCurve(ColorCurve=DataInterface() {propertyValues {
       AlphaCurve {keys [<at 0.0739 0 RCIM_Cubic>, {arriveTangent=3.9441, interpMode=RCIM_Cubic, time=0.2808, value=1}, <at 0.4005 1>, <at 0.6074 0>]}
       BlueCurve {keys [{arriveTangent=3.2911, interpMode=RCIM_Cubic, time=0.1753, value=0}, <at 0.6077 2>]}
       GreenCurve {keys [{arriveTangent=2.7598, interpMode=RCIM_Cubic, time=0.1753, value=0.8594}, <at 0.6077 2>]}
       RedCurve {keys [{arriveTangent=-0.1276, interpMode=RCIM_Cubic, time=0.1753, value=0.2064}, <at 0.6077 0>]}
      }})
    R1 Grey:LinearColor=&ColorFromCurve(ColorCurve=DataInterface() {propertyValues {
       AlphaCurve {keys [<at 0.1143 0>,<at 0.3088 1 RCIM_Cubic>,<at 0.6043 0 RCIM_Cubic>]}
       BlueCurve {keys [<at 0.3072 0.6562 RCIM_Cubic>]}
       GreenCurve {keys [<at 0.3072 0.6562 RCIM_Cubic>]}
       RedCurve {keys [<at 0.3072 0.6562 RCIM_Cubic>]}
      }})
    R1 Lime:LinearColor=&ColorFromCurve(ColorCurve=DataInterface() {propertyValues {
       AlphaCurve {keys [<at 0.0739 0 RCIM_Cubic>, {arriveTangent=3.9441, interpMode=RCIM_Cubic, time=0.2808, value=1}, <at 0.4005 1>, <at 0.6074 0>]}
       BlueCurve {keys [<at 0.091 0 RCIM_Cubic>,<at 0.6077 1.3543>]}
       GreenCurve {keys [<at 0.091 2 RCIM_Cubic>,<at 0.6077 2>]}
       RedCurve {keys [<at 0.091 1.1387 RCIM_Cubic>,<at 0.6077 0>]}
      }})
    R1 Orange:LinearColor=&ColorFromCurve(ColorCurve=DataInterface() {propertyValues {
       AlphaCurve {keys [<at 0.0739 0 RCIM_Cubic>, {arriveTangent=3.9441, interpMode=RCIM_Cubic, time=0.2808, value=1}, <at 0.4005 1>, <at 0.6074 0>]}
       BlueCurve {keys [<at 0.091 0 RCIM_Cubic>,<at 0.6077 0.2815>]}
       GreenCurve {keys [<at 0.091 0.7377 RCIM_Cubic>,<at 0.6077 1>]}
       RedCurve {keys [<at 0.091 2 RCIM_Cubic>,<at 0.6077 0.7462>]}
      }})
    R1 Pink:LinearColor=&ColorFromCurve(ColorCurve=DataInterface() {propertyValues {
       AlphaCurve {keys [<at 0.0739 0 RCIM_Cubic>, {arriveTangent=3.9441, interpMode=RCIM_Cubic, time=0.2808, value=1}, <at 0.4005 1>, <at 0.6074 0>]}
       BlueCurve {keys [<at 0.091 1.7582 RCIM_Cubic>,<at 0.3276 0.4062>]}
       GreenCurve {keys [<at 0.091 0.1064 RCIM_Cubic>,<at 0.3276 0.8298>]}
       RedCurve {keys [<at 0.091 2 RCIM_Cubic>,<at 0.3276 2>]}
      }})
    R1 Purple:LinearColor=&ColorFromCurve(ColorCurve=DataInterface() {propertyValues {
       AlphaCurve {keys [<at 0.1131 0 RCIM_Cubic>, {arriveTangent=3.9441, interpMode=RCIM_Cubic, time=0.2466, value=1}, <at 0.4005 1>, <at 0.6074 0>]}
       BlueCurve {keys [<at 0.2037 5 RCIM_Cubic>,<at 0.3276 0.709>]}
       GreenCurve {keys [<at 0.2037 0 RCIM_Cubic>,<at 0.3276 0.1062>]}
       RedCurve {keys [<at 0.2037 0.9897 RCIM_Cubic>,<at 0.3276 1>]}
      }})
    R1 Saffron:LinearColor=&ColorFromCurve(ColorCurve=DataInterface() {propertyValues {
       AlphaCurve {keys [<at 0.1131 0 RCIM_Cubic>, {arriveTangent=3.9441, interpMode=RCIM_Cubic, time=0.2466, value=1}, <at 0.4005 1>, <at 0.6074 0>]}
       BlueCurve {keys [<at 0.1542 0.875 RCIM_Cubic>,<at 0.3276 0.533>]}
       GreenCurve {keys [<at 0.1542 2.7603 RCIM_Cubic>,<at 0.3276 2>]}
       RedCurve {keys [<at 0.1542 3 RCIM_Cubic>,<at 0.3276 0.9975>]}
      }})
    R1 SkyBlue:LinearColor=&ColorFromCurve(ColorCurve=DataInterface() {propertyValues {
       AlphaCurve {keys [<at 0.1131 0 RCIM_Cubic>, {arriveTangent=3.9441, interpMode=RCIM_Cubic, time=0.2466, value=1}, <at 0.4005 1>, <at 0.6074 0>]}
       BlueCurve {keys [<at 0.1542 2 RCIM_Cubic>,<at 0.3276 1>]}
       GreenCurve {keys [<at 0.1542 1.715 RCIM_Cubic>,<at 0.3276 0.3668>]}
       RedCurve {keys [<at 0.1542 0.0312 RCIM_Cubic>,<at 0.3276 0.519>]}
      }})
    R1 TitaniumWhite:LinearColor=&ColorFromCurve(ColorCurve=DataInterface() {propertyValues {
       AlphaCurve {keys [<at 0.1131 0 RCIM_Cubic>, {arriveTangent=3.9441, interpMode=RCIM_Cubic, time=0.2466, value=1}, <at 0.4005 1>, <at 0.6074 0>]}
       BlueCurve {keys [<at 0.1542 2 RCIM_Cubic>,<at 0.3276 2>]}
       GreenCurve {keys [<at 0.1542 2 RCIM_Cubic>,<at 0.3276 2>]}
       RedCurve {keys [<at 0.1542 2 RCIM_Cubic>,<at 0.3276 2>]}
      }})
    R2 Black:LinearColor=&ColorFromCurve(ColorCurve=DataInterface() {propertyValues {
       AlphaCurve {keys [<at 0.0677 0 RCIM_Cubic>,<at 0.2668 0.8 RCIM_Cubic>,<at 1 0>]}
       BlueCurve {keys [<at 0.1091 0.001 RCIM_Cubic>,<at 0.2892 0.02 RCIM_Cubic>]}
       GreenCurve {keys [<at 0.1091 0.001 RCIM_Cubic>,<at 0.2892 0.02 RCIM_Cubic>]}
       RedCurve {keys [<at 0.1091 0.001 RCIM_Cubic>,<at 0.2892 0.02 RCIM_Cubic>]}
      }})
    R2 BurntSienna:LinearColor=&ColorFromCurve(ColorCurve=DataInterface() {propertyValues {
       AlphaCurve {keys [<at 0.1765 0 RCIM_Cubic>,<at 0.4052 0.6 RCIM_Cubic>]}
       BlueCurve {keys [<at 0.2419 0.0062 RCIM_Cubic>,<at 0.3865 0.0191 RCIM_Cubic>]}
       GreenCurve {keys [<at 0.2419 0.0287 RCIM_Cubic>,<at 0.3865 0.0573 RCIM_Cubic>]}
       RedCurve {keys [<at 0.2419 0.1771 RCIM_Cubic>,<at 0.3865 0.0504 RCIM_Cubic>]}
      }})
    R2 Cobalt:LinearColor=&ColorFromCurve(ColorCurve=DataInterface() {propertyValues {
       AlphaCurve {keys [<at 0.0817 0 RCIM_Cubic>,<at 0.4239 0.6 RCIM_Cubic>,<at 0.8485 0>]}
       BlueCurve {keys [<at 0.1753 0.3281 RCIM_Cubic>,<at 0.4954 0.224>]}
       GreenCurve {keys [<at 0.1753 0.0218 RCIM_Cubic>,<at 0.4954 0>]}
       RedCurve {keys [<at 0.1753 0 RCIM_Cubic>,<at 0.4954 0.1872>]}
      }})
    R2 Crimson:LinearColor=&ColorFromCurve(ColorCurve=DataInterface() {propertyValues {
       AlphaCurve {keys [<at 0.0817 0 RCIM_Cubic>,<at 0.3492 0.8 RCIM_Cubic>,<at 0.8485 0>]}
       BlueCurve {keys [<at 0.1753 0 RCIM_Cubic>,<at 0.4954 0.0137>]}
       GreenCurve {keys [<at 0.1753 0.0039 RCIM_Cubic>,<at 0.4954 0.0729>]}
       RedCurve {keys [<at 0.1753 0.2083 RCIM_Cubic>,<at 0.4954 0.0566>]}
      }})
    R2 DefaultColor:LinearColor=&ColorFromCurve(ColorCurve=DataInterface() {propertyValues {
       AlphaCurve {keys [<at 0.1672 0 RCIM_Cubic>,<at 0.3803 0.7 RCIM_Cubic>,<at 0.7334 0>]}
       BlueCurve {keys [<at 0.1753 0.1198 RCIM_Cubic>,<at 0.3508 0.1771>]}
       GreenCurve {keys [<at 0.1753 0.0381 RCIM_Cubic>,<at 0.3508 0.0839>]}
       RedCurve {keys [<at 0.1753 0.0461 RCIM_Cubic>,<at 0.3508 0.159>]}
      }})
    R2 ForestGreen:LinearColor=&ColorFromCurve(ColorCurve=DataInterface() {propertyValues {
       AlphaCurve {keys [
         <at 0.049 0>
         {arriveTangent=1.2606, interpMode=RCIM_Cubic, leaveTangent=1.2606, time=0.4161, value=0.8}
         <at 0.6836 0.8>
         <at 1 0 RCIM_Cubic>
        ]}
       BlueCurve {keys [<at 0 0.0334>,<at 0.4736 0.0922 RCIM_Cubic>]}
       GreenCurve {keys [<at 0 0.1823>,<at 0.4736 0.1198 RCIM_Cubic>]}
       RedCurve {keys [<at 0 0>,<at 0.4736 0 RCIM_Cubic>]}
      }})
    R2 Grey:LinearColor=&ColorFromCurve(ColorCurve=DataInterface() {propertyValues {
       AlphaCurve {keys [
         <at 0.147 0 RCIM_Cubic>
         {arriveTangent=3.1553, interpMode=RCIM_Cubic, leaveTangent=3.1553, time=0.2808, value=0.8}
         <at 0.4005 0.8>
         <at 0.6074 0>
        ]}
       BlueCurve {keys [<at 0.0708 0.2083 RCIM_Cubic>,<at 0.3508 0.0833>]}
       GreenCurve {keys [<at 0.0708 0.2083 RCIM_Cubic>,<at 0.3508 0.0833>]}
       RedCurve {keys [<at 0.0708 0.2083 RCIM_Cubic>,<at 0.3508 0.0833>]}
      }})
    R2 Lime:LinearColor=&ColorFromCurve(ColorCurve=DataInterface() {propertyValues {
       AlphaCurve {keys [
         <at 0.0412 0 RCIM_Cubic>
         {arriveTangent=1.6698, interpMode=RCIM_Cubic, leaveTangent=1.6698, time=0.2808, value=0.6}
         <at 0.4005 0.6>
         <at 0.6074 0>
        ]}
       BlueCurve {keys [<at 0.0941 0.0002 RCIM_Cubic>,<at 0.3803 0.1206>]}
       GreenCurve {keys [<at 0.0941 0.3438 RCIM_Cubic>,<at 0.3803 0.1927>]}
       RedCurve {keys [<at 0.0941 0.2305 RCIM_Cubic>,<at 0.3803 0.0384>]}
      }})
    R2 Orange:LinearColor=&ColorFromCurve(ColorCurve=DataInterface() {propertyValues {
       AlphaCurve {keys [
         <at 0.0412 0 RCIM_Cubic>
         {arriveTangent=1.6698, interpMode=RCIM_Cubic, leaveTangent=1.6698, time=0.1552, value=0.5}
         <at 0.4005 0.6>
         <at 0.6074 0>
        ]}
       BlueCurve {keys [<at 0.1583 0 RCIM_Cubic>,<at 0.3803 0>]}
       GreenCurve {keys [<at 0.1583 0.2274 RCIM_Cubic>,<at 0.3803 0.1458>]}
       RedCurve {keys [<at 0.1583 0.5469 RCIM_Cubic>,<at 0.3803 0.1076>]}
      }})
    R2 Pink:LinearColor=&ColorFromCurve(ColorCurve=DataInterface() {propertyValues {
       AlphaCurve {keys [
         <at 0.0412 0 RCIM_Cubic>
         {arriveTangent=1.6698, interpMode=RCIM_Cubic, leaveTangent=1.6698, time=0.1552, value=0.5}
         <at 0.4005 0.6>
         <at 0.6074 0>
        ]}
       BlueCurve {keys [<at 0.1583 0.2922 RCIM_Cubic>,<at 0.3803 0.0526>]}
       GreenCurve {keys [<at 0.1583 0.0423 RCIM_Cubic>,<at 0.3803 0.0787>]}
       RedCurve {keys [<at 0.1583 0.4271 RCIM_Cubic>,<at 0.3803 0.1094>]}
      }})
    R2 Purple:LinearColor=&ColorFromCurve(ColorCurve=DataInterface() {propertyValues {
       AlphaCurve {keys [
         <at 0.0412 0 RCIM_Cubic>
         {arriveTangent=1.6698, interpMode=RCIM_Cubic, leaveTangent=1.6698, time=0.1552, value=0.5}
         <at 0.4005 0.6>
         <at 0.6074 0>
        ]}
       BlueCurve {keys [<at 0.1583 1 RCIM_Cubic>,<at 0.3803 0.08>]}
       GreenCurve {keys [<at 0.1583 0 RCIM_Cubic>,<at 0.3803 0.0166>]}
       RedCurve {keys [<at 0.1583 0.287 RCIM_Cubic>,<at 0.3803 0.1094>]}
      }})
    R2 Saffron:LinearColor=&ColorFromCurve(ColorCurve=DataInterface() {propertyValues {
       AlphaCurve {keys [
         <at 0.0412 0 RCIM_Cubic>
         {arriveTangent=1.6698, interpMode=RCIM_Cubic, leaveTangent=1.6698, time=0.1552, value=0.5}
         <at 0.4005 0.6>
         <at 0.6074 0>
        ]}
       BlueCurve {keys [<at 0.1583 0 RCIM_Cubic>,<at 0.3803 0.0873>]}
       GreenCurve {keys [<at 0.1583 0.8117 RCIM_Cubic>,<at 0.3803 0.2396>]}
       RedCurve {keys [<at 0.1583 0.875 RCIM_Cubic>,<at 0.3803 0.1074>]}
      }})
    R2 SkyBlue:LinearColor=&ColorFromCurve(ColorCurve=DataInterface() {propertyValues {
       AlphaCurve {keys [
         <at 0.0412 0 RCIM_Cubic>
         {arriveTangent=1.6698, interpMode=RCIM_Cubic, leaveTangent=1.6698, time=0.1552, value=0.5}
         <at 0.4005 0.6>
         <at 0.6074 0>
        ]}
       BlueCurve {keys [<at 0.1583 0.4167 RCIM_Cubic>,<at 0.3803 0.125>]}
       GreenCurve {keys [<at 0.1583 0.2756 RCIM_Cubic>,<at 0.3803 0.041>]}
       RedCurve {keys [<at 0.1583 0.0434 RCIM_Cubic>,<at 0.3803 0.0705>]}
      }})
    R2 TitaniumWhite:LinearColor=&ColorFromCurve(ColorCurve=DataInterface() {propertyValues {
       AlphaCurve {keys [
         <at 0.0412 0 RCIM_Cubic>
         {arriveTangent=1.6698, interpMode=RCIM_Cubic, leaveTangent=1.6698, time=0.1552, value=0.5}
         <at 0.4005 0.6>
         <at 0.6074 0>
        ]}
       BlueCurve {keys [<at 0.1583 0.6667 RCIM_Cubic>,<at 0.3803 0.2031>]}
       GreenCurve {keys [<at 0.1583 0.6667 RCIM_Cubic>,<at 0.3803 0.2031>]}
       RedCurve {keys [<at 0.1583 0.6667 RCIM_Cubic>,<at 0.3803 0.2031>]}
      }})
    R3 DefaultColor:LinearColor=&ColorFromCurve(ColorCurve=DataInterface() {propertyValues {
       AlphaCurve {
        defaultValue=0
        keys [
         <at 0 0 RCIM_Cubic>
         <at 0.2 1>
         {
          arriveTangent=-0.0405
          arriveTangentWeight=0.2004
          interpMode=RCIM_Cubic
          leaveTangent=-0.0405
          leaveTangentWeight=0.0998
          tangentMode=RCTM_User
          time=0.7009
          value=1
         }
         {arriveTangent=-8.0089, interpMode=RCIM_Cubic, leaveTangent=-8.0089, tangentMode=RCTM_User, time=1, value=0}
        ]
       }
       BlueCurve {keys [<at 0.1753 1.5 RCIM_Cubic>,<at 0.6077 2>]}
       GreenCurve {keys [<at 0.1753 0.6704 RCIM_Cubic>,<at 0.6077 0.6146>]}
       RedCurve {keys [<at 0.1753 0.7521 RCIM_Cubic>,<at 0.6077 1.7304>]}
      }})
    Saffron=&ColorFromCurve(ColorCurve=DataInterface() {propertyValues {
       AlphaCurve {keys [
         <at 0.0789 0 RCIM_Cubic>
         {arriveTangent=2.7665, interpMode=RCIM_Cubic, leaveTangent=2.7665, time=0.2317, value=1}
         <at 0.4403 1>
         <at 0.7668 0>
        ]}
       BlueCurve {keys [<at 0.2063 0 RCIM_Cubic>,<at 0.5327 0 RCIM_Cubic>]}
       GreenCurve {keys [<at 0.2063 0.7424 RCIM_Cubic>,<at 0.5327 0.15 RCIM_Cubic>]}
       RedCurve {keys [<at 0.2063 0.875 RCIM_Cubic>,<at 0.5327 0.0779 RCIM_Cubic>]}
      }})
    SkyBlue=&ColorFromCurve(ColorCurve=DataInterface() {propertyValues {
       AlphaCurve {keys [<at 0.0789 0 RCIM_Cubic>, {arriveTangent=2.5883, interpMode=RCIM_Cubic, time=0.2349, value=0.8}, <at 0.4301 1>, <at 0.8209 0>]}
       BlueCurve {keys [<at 0.1022 0.4792 RCIM_Cubic>,<at 0.5215 0.1 RCIM_Cubic>]}
       GreenCurve {keys [<at 0.1022 0.4551 RCIM_Cubic>,<at 0.5215 0.018 RCIM_Cubic>]}
       RedCurve {keys [<at 0.1022 0.0393 RCIM_Cubic>,<at 0.5215 0.0466 RCIM_Cubic>]}
      }})
    TitaniumWhite=&ColorFromCurve(ColorCurve=DataInterface() {propertyValues {
       AlphaCurve {keys [
         <at 0.0852 0 RCIM_Cubic>
         {arriveTangent=2.8998, interpMode=RCIM_Cubic, leaveTangent=2.8998, time=0.2078, value=0.5}
         <at 0.4301 1>
         <at 0.6856 0>
        ]}
       BlueCurve {keys [<at 0.1022 1 RCIM_Cubic>,<at 0.5391 0.11 RCIM_Cubic>]}
       GreenCurve {keys [<at 0.1022 1 RCIM_Cubic>,<at 0.5391 0.11 RCIM_Cubic>]}
       RedCurve {keys [<at 0.1022 1 RCIM_Cubic>,<at 0.5391 0.11 RCIM_Cubic>]}
      }})
    Use R1 Painted Overrides=true
    Use R2 Painted Overrides=true
    Use R3 Painted Overrides=true
   ) {module=VC_PaintedVariantsOverride001}
   &ScaleSpriteSize(
    Uniform Curve Scale=1.2
    Uniform Curve Sprite Scale=DataInterface() {propertyValues {Curve {
       defaultValue=5
       keys [
        {
         arriveTangent=-0.0786
         arriveTangentWeight=0.4631
         interpMode=RCIM_Cubic
         leaveTangent=-0.0786
         leaveTangentWeight=0.4631
         tangentMode=RCTM_User
         tangentWeightMode=RCTWM_WeightedBoth
         time=0
         value=0.33
        }
        {
         arriveTangent=1.2246
         arriveTangentWeight=0.5053
         interpMode=RCIM_Cubic
         leaveTangent=1.2246
         leaveTangentWeight=0.5053
         tangentMode=RCTM_User
         tangentWeightMode=RCTWM_WeightedBoth
         time=1
         value=0.85
        }
       ]
      }}}
   )
   &InheritSourceMovement(ApplyForce=false, ApplyPosition=true, ApplyPositionFalloff=false)
   &AccelerationForce(Acceleration=<0 0 200>)
   &SolveForcesAndVelocity(Clamp Velocity=false, Limit Acceleration=false)
   &SubUVAnimation(
    End Frame Range Override=22
    "Sprite Renderer "=DataInterface() {propertyValues {SpriteRenderer=&self:NE_CS_MainTrail_0.NiagaraSpriteRendererProperties_6}}
    SubUV Animation Mode:ENiagara_SubUVLookupModeV2=::NewEnumerator1
    SubUV Curve=DataInterface() {propertyValues {Curve {
       defaultValue=0
       keys [
        {
         arriveTangent=0.3386
         arriveTangentWeight=0.2442
         interpMode=RCIM_Cubic
         leaveTangent=0.3386
         leaveTangentWeight=0.2442
         tangentMode=RCTM_User
         tangentWeightMode=RCTWM_WeightedBoth
         time=0
         value=0
        }
        {interpMode=RCIM_Cubic, tangentMode=RCTM_User, time=1, value=1}
       ]
      }}}
    UseEndFrame=true
    UseStartFrame=true
   ) {module=SubUVAnimation003}
   /SetVariables(Particles.HighlightColor:LinearColor=&HueShiftLinearColor(
     Hue=0
     Saturation=1
     Value=&FloatFromCurve(
      FloatCurve=DataInterface() {propertyValues {Curve {keys [<at 0 4 RCIM_Constant>,<at 0.07 1>]}}}
      CurveIndex=&RandomRangeFloat(Minimum=0, Maximum=1)
     )
     Linear Color=$Particles.Color
    )) {isSetParameters=true}
  ]
  rendererClasses [&NiagaraSpriteRendererProperties]
  renderers [{rendererClass=&NiagaraSpriteRendererProperties}]
 }
}
summary {
 systemName=NS_Boost_Colorsmoke
 emitters [
  {emitterName=NE_CS_Burst, rendererClasses [&NiagaraSpriteRendererProperties]}
  {emitterName=NE_CS_MainTrail, rendererClasses [&NiagaraSpriteRendererProperties]}
 ]
 userVariables [
  User.bIsFrontendPreview {type=NiagaraBool}
  User.bIsPainted {type=NiagaraBool}
  User.ColorVariantID {type=NiagaraInt32}
  User.DTCarOffset {type=Vector3f, description="", defaultValue {struct=Vector3f, value=<0 0 0>}}
  User.PaintedColor {type=LinearColor, description="", defaultValue {struct=LinearColor, value=<rgba 1 1 1 1>}}
  User.PaintedScalar {type=NiagaraFloat, defaultValue=1, description=""}
  User.PortCount {type=NiagaraInt32, defaultValue=2}
 ]
}
